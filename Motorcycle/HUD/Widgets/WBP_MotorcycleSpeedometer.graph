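# Motorcycle dashboard HUD widget, extending WBP_VehicleGaugeBase, with speedometer, fuel gauge, low-fuel warning, rechargeable boost meter and vehicle name popup.
# On visibility changes it binds or unbinds vehicle delegates, showing a jump meter only for DirtBike-tagged vehicles.
# Jump charge drives PercentFill on dynamic materials, and boost events play meter and spark effects.
# Inventory mode scales the gauge via separate touch and non-touch animations, and names slide in, then fade on a timer.
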WidgetBlueprint /Motorcycle/HUD/Widgets/WBP_MotorcycleSpeedometer
animations [
 NameIntro {
  binding_count=1
  length_seconds=0.35
  object=&self:NameIntro
  sequence {bindings [Text_ItemName {
     id=#1
     possessed_class=CommonTextBlock
     tracks [
      {
       type=MovieSceneFloatTrack
       display_name=Render Opacity
       sections [{type=MovieSceneFloatSection, end_frame=7, start_frame=0, channels [None {type=MovieSceneScriptingFloatChannel, keys [[1.5157,0],[7,1]]}]}]
      }
      {
       type=MovieScene2DTransformTrack
       sections [{
         type=MovieScene2DTransformSection
         end_frame=7
         start_frame=0
         channels [Translation.X {type=MovieSceneScriptingFloatChannel, keys [[0,63.2905],[7,0]]}]
        }]
      }
     ]
    }]}
 }
 NameOutro {
  binding_count=1
  length_seconds=1.0019
  object=&self:NameOutro
  sequence {bindings [Text_ItemName {
     id=#2
     possessed_class=CommonTextBlock
     tracks [
      {
       type=MovieSceneFloatTrack
       display_name=Render Opacity
       sections [{type=MovieSceneFloatSection, channels [None {type=MovieSceneScriptingFloatChannel, keys [[0,1],[20.037,0]]}]}]
      }
      {
       type=MovieScene2DTransformTrack
       sections [{type=MovieScene2DTransformSection, channels [Translation.X {type=MovieSceneScriptingFloatChannel, keys [[0,0],[20.037,-20]]}]}]
      }
     ]
    }]}
 }
 JumpFull {
  binding_count=1
  length_seconds=1.0861
  object=&self:JumpFull
  sequence {bindings [Image_JumpIcon {
     id=#3
     possessed_class=Image
     tracks [{
       type=MovieScene2DTransformTrack
       sections [{
         type=MovieScene2DTransformSection
         channels [
          Translation.X {type=MovieSceneScriptingFloatChannel, const [90,1]}
          Translation.Y {type=MovieSceneScriptingFloatChannel, const [140,1]}
          Angle {type=MovieSceneScriptingFloatChannel, const [0,1]}
          Scale.X {type=MovieSceneScriptingFloatChannel, keys [[0,1],[4.1223,1.2],[21.7223,1]]}
          Scale.Y {type=MovieSceneScriptingFloatChannel, keys [[0,1],[4.1223,1.2],[21.7223,1]]}
          Shear.X {type=MovieSceneScriptingFloatChannel, const [0,1]}
          Shear.Y {type=MovieSceneScriptingFloatChannel, const [0,1]}
         ]
        }]
      }]
    }]}
 }
 TransitionToCombat {
  binding_count=1
  length_seconds=0.25
  object=&self:TransitionToCombat
  sequence {bindings [Overlay_Gauge {
     id=#4
     possessed_class=Overlay
     tracks [{
       type=MovieScene2DTransformTrack
       sections [{
         type=MovieScene2DTransformSection
         end_frame=5
         start_frame=0
         channels [
          Scale.X {type=MovieSceneScriptingFloatChannel, keys [[0,0.7],[5,0.5]]}
          Scale.Y {type=MovieSceneScriptingFloatChannel, keys [[0,0.7],[5,0.5]]}
         ]
        }]
      }]
    }]}
 }
 TransitionToCombat_Mobile {
  binding_count=1
  length_seconds=0.25
  object=&self:TransitionToCombat_Mobile
  sequence {bindings [Overlay_Gauge {
     id=#4
     possessed_class=Overlay
     tracks [{
       type=MovieScene2DTransformTrack
       sections [{
         type=MovieScene2DTransformSection
         end_frame=5
         start_frame=0
         channels [
          Scale.X {type=MovieSceneScriptingFloatChannel, keys [[0,1],[5,0.85]]}
          Scale.Y {type=MovieSceneScriptingFloatChannel, keys [[0,1],[5,0.85]]}
         ]
        }]
      }]
    }]}
 }
 BoostEngage {
  binding_count=2
  length_seconds=5
  object=&self:BoostEngage
  sequence {bindings [
    Image_BoostMeter {
     id=#5
     possessed_class=Image
     tracks [{
       type=MovieSceneWidgetMaterialTrack
       sections [{type=MovieSceneParameterSection, channels [ColorOverrideScalar {type=MovieSceneScriptingFloatChannel, keys [[0,0],[5,1]]}]}]
      }]
    }
    Overlay_BoostSparks {
     id=#6
     possessed_class=Overlay
     tracks [{
       type=MovieSceneFloatTrack
       display_name=Render Opacity
       sections [{type=MovieSceneFloatSection, end_frame=5, start_frame=0, channels [None {type=MovieSceneScriptingFloatChannel, keys [[0,0],[5,1]]}]}]
      }]
    }
   ]}
 }
 BoostFull {
  binding_count=1
  length_seconds=5
  object=&self:BoostFull
  sequence {bindings [Image_BoostIcon {
     id=#7
     possessed_class=Image
     tracks [{
       type=MovieSceneWidgetMaterialTrack
       sections [{
         type=MovieSceneParameterSection
         channels [
          "TimeOffset (S)" {type=MovieSceneScriptingFloatChannel, keys [[0,0],[21.7223,1],[88.9073,1]]}
          IconScale {type=MovieSceneScriptingFloatChannel, keys [[0,0.2476],[4.1223,0.45],[21.7223,0.2476]]}
         ]
        }]
      }]
    }]}
 }
]
graphs [
 EventGraph {nodes [
   StopBoostStuff() {is=title K2Node_CustomEvent}
   > Branch(Condition=Get BoostEffectsVisible() -> (BoostEffectsVisible:bool))
   > PlayAnimationReverse(InAnimation=Get BoostEngage() -> (BoostEngage:Widget Animation Object Reference), PlaybackSpeed=0.5)
   > SetVisibility(self=Get Overlay_BoostSparks() -> (Overlay_BoostSparks:Overlay Object Reference), InVisibility=Hidden)
   > Set BoostEffectsVisible(BoostEffectsVisible:bool=false) -> (Output_Get:bool)
   n1 = OnPopUpOnSwitchDone() {is=title K2Node_CustomEvent}
   > StopAnimation(InAnimation=Get NameIntro() -> (NameIntro:Widget Animation Object Reference))
   > PlayAnimation(InAnimation=Get NameOutro() -> (NameOutro:Widget Animation Object Reference))
   n2 = OnInventoryModeChanged!() -> (OutputDelegate, then, IsInInventoryMode:bool=~) {is=title K2Node_CustomEvent}
   > n3 = Set Is in Inventory Mode(Is in Inventory Mode:bool=@n2.IsInInventoryMode) -> (Output_Get:bool)
   > n4 = Branch(Condition=@n3)
   n5 = Select(
    Option 0:Widget Animation Object Reference=Get TransitionToCombat() -> (TransitionToCombat:Widget Animation Object Reference)
    Option 1:Widget Animation Object Reference=Get TransitionToCombat_Mobile() -> (TransitionToCombat_Mobile:Widget Animation Object Reference)
    Index=IsUsingTouch(PlayerController=GetOwningPlayer())
   ) -> (ReturnValue:Widget Animation Object Reference)
   > PlayAnimationForward(InAnimation=@n5)
   PlayAnimationReverse(execute=@n4.else, InAnimation=@n5)
   n6 = OnChargingJump!() -> (OutputDelegate, then, JumpChargeFactor:float=~) {is=title K2Node_CustomEvent}
   > SetScalarParameterValue(
    self=[
     Get JumpGlowMaterial() -> (JumpGlowMaterial:Material Instance Dynamic Object Reference)
     Get ImageJumpMaterial() -> (ImageJumpMaterial:Material Instance Dynamic Object Reference)
    ]
    ParameterName=PercentFill
    Value=@n6.JumpChargeFactor
   )
   > Branch(Condition="float >= float"(A=@n6.JumpChargeFactor, B=1))
   > PlayAnimationForward(InAnimation=Get JumpFull() -> (JumpFull:Widget Animation Object Reference))
   n7 = OnCancelChargingJump() {is=title K2Node_CustomEvent}
   SetScalarParameterValue(execute=[@n7.then, @n8.then], self=[Get ImageJumpMaterial(), Get JumpGlowMaterial()], ParameterName=PercentFill, Value=0)
   n8 = OnJump!() -> (OutputDelegate, then, JumpChargeFactor:float=~) {is=title K2Node_CustomEvent}
   Event PreConstruct()
   > Branch(Condition=IsMobileGame())
   > RemoveWidgetTagsToHideWhenVisible!(
    execute
    self:Fort Athena Vehicle Dashboard Widget Object Reference
    TagToRemove:Gameplay Tag Container Structure=(GameplayTags=((TagName="HUD.TopLeft.PickupManager")))
   ) -> (then)
   n9 = Event OnWidgetVisibilityChanged!() -> (OutputDelegate, then, bNewVisibility:bool) {is=title K2Node_Event}
   > n10 = Cast To FortMotorcycleVehicle!(
    execute
    Object=GetVehicle!(self:Fort Athena Vehicle Dashboard Widget Object Reference) -> (ReturnValue:Fort Athena Vehicle Object Reference)
   ) -> (then, CastFailed, AsFort Motorcycle Vehicle:Fort Motorcycle Vehicle Object Reference) {is=title K2Node_DynamicCast}
   > n11 = Branch(Condition=@n9.bNewVisibility)
   n12 = HasTag!(
    TagContainer:"Gameplay Tag Container Structure (by ref)"=GetVehicleTags!(self:Fort Vehicle Interface Interface=@n10) -> (ReturnValue:"Gameplay Tag Container Structure (by ref)")
    Tag:Gameplay Tag Structure=(TagName="Vehicle.Motorcycle.DirtBike")
    bExactMatch:bool=true
   ) -> (ReturnValue:bool)
   > SetVisibility(
    self=Get Overlay_JumpMeter() -> (Overlay_JumpMeter:Overlay Object Reference)
    InVisibility=Select(Option 0:ESlateVisibility Enum=Collapsed, Option 1:ESlateVisibility Enum=SelfHitTestInvisible, Index=@n12) -> (ReturnValue:ESlateVisibility Enum)
   )
   > n13 = Branch(Condition=@n12)
   n14 = Create Event()
   > Bind Event to On Charging Jump Delegate(self:Fort Motorcycle Vehicle Object Reference=@n10, Delegate=@n14) {is=title K2Node_AddDelegate}
   n15 = Create Event()
   > Bind Event to On Cancel Charging Jump Delegate(self:Fort Motorcycle Vehicle Object Reference=@n10, Delegate=@n15) {is=title K2Node_AddDelegate}
   n16 = Create Event()
   > n17 = Bind Event to On Jump Delegate(self:Fort Motorcycle Vehicle Object Reference=@n10, Delegate=@n16) {is=title K2Node_AddDelegate}
   n18 = Create Event()
   Assign On Inventory Mode Changed Delegate!(execute:Exec=[@n17.then, @n13.else], self:Fort Motorcycle Vehicle Object Reference=@n10, Delegate:"Delegate (by ref)"=@n18) -> (then:Exec) {is=title K2Node_AssignDelegate}
   > SetScalarParameterValue(self=[Get JumpGlowMaterial(), Get ImageJumpMaterial()], ParameterName=PercentFill, Value=0)
   StopBoostStuff(execute=@n11.else) {is=title K2Node_CallFunction}
   > n19 = Branch(Condition=@n12)
   > Unbind Event from On Charging Jump Delegate(self:Fort Motorcycle Vehicle Object Reference=@n10, Delegate=@n14) {is=title K2Node_RemoveDelegate}
   > Unbind Event from On Cancel Charging Jump Delegate(self:Fort Motorcycle Vehicle Object Reference=@n10, Delegate=@n15) {is=title K2Node_RemoveDelegate}
   > n20 = Unbind Event from On Jump Delegate(self:Fort Motorcycle Vehicle Object Reference=@n10, Delegate=@n16) {is=title K2Node_RemoveDelegate}
   Unbind Event from On Inventory Mode Changed Delegate(execute=[@n20.then, @n19.else], self:Fort Motorcycle Vehicle Object Reference=@n10, Delegate=@n18) {is=title K2Node_RemoveDelegate}
   n21 = Get TransitionToCombat()
   > Branch(Condition=OR Boolean(A=Get Is in Inventory Mode() -> (Is in Inventory Mode:bool), B=IsAnimationPlaying(InAnimation=@n21)))
   > Set Is in Inventory Mode(Is in Inventory Mode=false)
   n22 = Get TransitionToCombat()
   > PlayAnimation(InAnimation=@n22, StartAtTime=GetEndTime(self=@n22), PlayMode=Reverse)
   n23 = Event SetVehicleHasInfiniteFuel!() -> (OutputDelegate, then, bInfiniteFuel:bool) {is=title K2Node_Event}
   > Parent: SetVehicleHasInfiniteFuel!(execute, bInfiniteFuel:bool=@n23.bInfiniteFuel) -> (then) {is=title K2Node_CallParentFunction}
   > ShowInfiniteFuel!(
    execute
    self:Fort Vehicle Fuel Gauge Object Reference=Get Widget_FuelGauge() -> (Widget_FuelGauge:Fort Vehicle Fuel Gauge Object Reference)
    bHasInfiniteFuel:bool=@n23.bInfiniteFuel
   ) -> (then)
   > UpdateFuelChanged!(execute, self:Fort Vehicle Fuel Gauge Object Reference=Get Widget_FuelGauge(), FuelPercent:float=1, bIsFuelLow:bool=false) -> (then)
   Event OnBoostStarted() {is=title K2Node_Event}
   > n24 = Branch(Condition=Get BoostEffectsVisible())
   PlayAnimationForward(execute=@n24.else, InAnimation=Get BoostEngage(), PlaybackSpeed=4)
   > SetVisibility(self=Get Overlay_BoostSparks(), InVisibility=HitTestInvisible)
   > Set BoostEffectsVisible(BoostEffectsVisible=true)
   Event OnBoostFinished() {is=title K2Node_Event}
   > StopBoostStuff()
   n25 = Event SetVehicleUsesRechargableBoost!() -> (OutputDelegate, then, bUsesRechargeableBoost:bool) {is=title K2Node_Event}
   > SetVisibility(
    self=[Get Spacer_Boost() -> (Spacer_Boost:Spacer Object Reference), Get OverlayBoostMeter() -> (OverlayBoostMeter:Overlay Object Reference)]
    InVisibility=Select(Option 0=Collapsed, Option 1=HitTestInvisible, Index=@n25.bUsesRechargeableBoost)
   )
   Event Construct()
   > SetVisibility(InVisibility=Collapsed)
   > n26 = GetDynamicMaterial(self=Get Image_JumpMeter() -> (Image_JumpMeter:Image Object Reference))
   > Set ImageJumpMaterial(ImageJumpMaterial:Material Instance Dynamic Object Reference=@n26) -> (Output_Get:Material Instance Dynamic Object Reference)
   > n27 = GetDynamicMaterial(self=Get Image_JumpGlow() -> (Image_JumpGlow:Image Object Reference))
   > Set JumpGlowMaterial(JumpGlowMaterial:Material Instance Dynamic Object Reference=@n27) -> (Output_Get:Material Instance Dynamic Object Reference)
   Event OnPlayerFacingNameUpdated() {is=title K2Node_Event}
   > SetVisibility(self=Get Overlay_CarInfo() -> (Overlay_CarInfo:Overlay Object Reference), InVisibility=SelfHitTestInvisible)
   > Set Timer by Event(Delegate=@n1, Time=2, bLooping=false)
   > StopAnimation(InAnimation=Get NameOutro())
   > PlayAnimation(InAnimation=Get NameIntro())
   PlayAnimationForward(InAnimation=Get BoostFull() -> (BoostFull:Widget Animation Object Reference))
   Event OnRechargableBoostFilled() {is=title K2Node_Event}
  ]}
 GetWarningWidget {nodes [
   GetWarningWidget() {is=title K2Node_FunctionEntry}
   > Return Node!(
    execute
    WarningWidget:WBP Vehicle Warning Object Reference=Get Widget_LowFuelWarning() -> (Widget_LowFuelWarning:WBP Vehicle Warning Object Reference)
   )
  ]}
]
variables [
 BoostEffectsVisible {type=boolean}
 ImageJumpMaterial {type=MaterialInstanceDynamic}
 Is in Inventory Mode {type=boolean}
 JumpGlowMaterial {type=MaterialInstanceDynamic}
]
widgets {
 info {
  inheritedWidgetCount=10
  parentClass=/Game/Athena/HUD/Vehicles/Dashboard/WBP_VehicleGaugeBase.WBP_VehicleGaugeBase_C
  rootWidgetClass=&Overlay
  widgetCount=30
 }
 widgets [
  Overlay_Gauge {
   widgetClassPath=&Overlay
   bInherited=true
   widget=Overlay_Gauge
   properties {visibility=HitTestInvisible, renderTransformPivot=<0 1>}
  }
  Widget_LowFuelWarning {
   widgetClassPath=/Game/Athena/HUD/Vehicles/Dashboard/WBP_Vehicle_Warning
   bIsVariable=true
   parent=Overlay_Gauge
   widget=Widget_LowFuelWarning
   properties {visibility=Collapsed, renderTransform {angle=0, scale=<1 1>, shear=<0 0>, translation=<-120 -183>}}
   slot {class=OverlaySlot, properties {verticalAlignment=VAlign_Bottom}}
  }
  HorizontalBox_FuelCluster {
   widgetClassPath=&HorizontalBox
   bInherited=true
   parent=Overlay_Gauge
   widget=HorizontalBox_FuelCluster
   properties {renderTransform {angle=0, scale=<1 1>, shear=<0 0>, translation=<0 -52>}}
   slot {class=OverlaySlot, properties {verticalAlignment=VAlign_Bottom}}
  }
  Spacer_Boost {
   widgetClassPath=&Spacer
   bIsVariable=true
   parent=HorizontalBox_FuelCluster
   widget=Spacer_Boost
   properties {visibility=Collapsed, size=<65 1>}
   slot {class=HorizontalBoxSlot, properties {horizontalAlignment=HAlign_Right}}
  }
  Widget_FuelGauge {
   widgetClassPath=/Game/Athena/HUD/Vehicles/Dashboard/WBP_Vehicle_FuelGauge
   bInherited=true
   bIsVariable=true
   parent=HorizontalBox_FuelCluster
   widget=Widget_FuelGauge
   slot {class=HorizontalBoxSlot, properties {horizontalAlignment=HAlign_Right, verticalAlignment=VAlign_Center}}
  }
  Spacer_Jump {
   widgetClassPath=&Spacer
   bIsVariable=true
   parent=HorizontalBox_FuelCluster
   widget=Spacer_Jump
   properties {visibility=HitTestInvisible, size=<65 1>}
   slot {class=HorizontalBoxSlot, properties {horizontalAlignment=HAlign_Right}}
  }
  Overlay_SpeedCluster {
   widgetClassPath=&Overlay
   parent=Overlay_Gauge
   widget=Overlay_SpeedCluster
   properties {renderTransform {angle=0, scale=<1 1>, shear=<0 0>, translation=<147 16>}}
   slot {class=OverlaySlot, properties {verticalAlignment=VAlign_Bottom, padding=<ltrb 0 0 45 0>}}
  }
  Overlay_JumpMeter {
   widgetClassPath=&Overlay
   bIsVariable=true
   parent=Overlay_SpeedCluster
   widget=Overlay_JumpMeter
   properties {visibility=HitTestInvisible, renderTransform {angle=0, scale=<1 1>, shear=<0 0>, translation=<-50 -10>}}
   slot {class=OverlaySlot, properties {horizontalAlignment=HAlign_Center, verticalAlignment=VAlign_Center, padding=<ltrb -200 -200 -200 -200>}}
  }
  Image_JumpGlow {
   widgetClassPath=&Image
   bIsVariable=true
   parent=Overlay_JumpMeter
   widget=Image_JumpGlow
   properties {
    visibility=SelfHitTestInvisible
    brush {
     bIsDynamicallyLoaded=false
     drawAs=Image
     imageType=NoImage
     mirroring=NoMirror
     resourceName=None
     resourceObject=/Game/Athena/HUD/Vehicles/Materials/M_UI_ValetMeterGlow_Inst
     tiling=NoTile
     imageSize=<340 340>
     margin=<ltrb 0 0 0 0>
     outlineSettings {
      bUseBrushTransparency=false
      roundingType=HalfHeightRadius
      width=0
      color {colorUseRule=UseColor_Specified, specifiedColor=<rgba 0 0 0 0>}
      cornerRadii=<0 0 0 0>
     }
     tintColor {colorUseRule=UseColor_Specified, specifiedColor=<rgba 1 1 1 1>}
     uVRegion {bIsValid=false, max=<0 0>, min=<0 0>}
    }
   }
   slot {class=OverlaySlot, properties {horizontalAlignment=HAlign_Center, verticalAlignment=VAlign_Center, padding=<ltrb -200 -200 -200 -200>}}
  }
  Image_JumpMeter {
   widgetClassPath=&Image
   bIsVariable=true
   parent=Overlay_JumpMeter
   widget=Image_JumpMeter
   properties {
    visibility=SelfHitTestInvisible
    brush {
     bIsDynamicallyLoaded=false
     drawAs=Image
     imageType=NoImage
     mirroring=NoMirror
     resourceName=None
     resourceObject=/Game/Athena/HUD/Vehicles/Materials/M_UI_ValetSimpleMeter_Boost
     tiling=NoTile
     imageSize=<320 320>
     margin=<ltrb 0 0 0 0>
     outlineSettings {
      bUseBrushTransparency=false
      roundingType=HalfHeightRadius
      width=0
      color {colorUseRule=UseColor_Specified, specifiedColor=<rgba 0 0 0 0>}
      cornerRadii=<0 0 0 0>
     }
     tintColor {colorUseRule=UseColor_Specified, specifiedColor=<rgba 1 1 1 1>}
     uVRegion {bIsValid=false, max=<0 0>, min=<0 0>}
    }
   }
   slot {class=OverlaySlot}
  }
  Image_JumpIcon {
   widgetClassPath=&Image
   parent=Overlay_JumpMeter
   widget=Image_JumpIcon
   properties {
    visibility=SelfHitTestInvisible
    brush {
     bIsDynamicallyLoaded=false
     drawAs=Image
     imageType=NoImage
     mirroring=NoMirror
     resourceName=None
     resourceObject=/Motorcycle/UI/T_UI_Motorcycle_Jump
     tiling=NoTile
     imageSize=<48 48>
     margin=<ltrb 0 0 0 0>
     outlineSettings {
      bUseBrushTransparency=false
      roundingType=HalfHeightRadius
      width=0
      color {colorUseRule=UseColor_Specified, specifiedColor=<rgba 0 0 0 0>}
      cornerRadii=<0 0 0 0>
     }
     tintColor {colorUseRule=UseColor_Specified, specifiedColor=<rgba 1 1 1 1>}
     uVRegion {bIsValid=false, max=<0 0>, min=<0 0>}
    }
    renderTransform {angle=0, scale=<1 1>, shear=<0 0>, translation=<90 140>}
   }
   slot {class=OverlaySlot, properties {verticalAlignment=VAlign_Center, padding=<ltrb -64 -64 -64 -64>}}
  }
  Overlay_Speedometer {widgetClassPath=&Overlay, parent=Overlay_SpeedCluster, widget=Overlay_Speedometer, slot {class=OverlaySlot}}
  Image_Speedometer {
   widgetClassPath=&Image
   bInherited=true
   bIsVariable=true
   parent=Overlay_Speedometer
   widget=Image_Speedometer
   properties {
    visibility=SelfHitTestInvisible
    brush {
     bIsDynamicallyLoaded=false
     drawAs=Image
     imageType=NoImage
     mirroring=NoMirror
     resourceName=None
     resourceObject=/Game/Athena/HUD/Vehicles/Materials/M_UI_ValetDash_Speedometer_Inst
     tiling=NoTile
     imageSize=<330 330>
     margin=<ltrb 0 0 0 0>
     outlineSettings {
      bUseBrushTransparency=false
      roundingType=HalfHeightRadius
      width=0
      color {colorUseRule=UseColor_Specified, specifiedColor=<rgba 0 0 0 0>}
      cornerRadii=<0 0 0 0>
     }
     tintColor {colorUseRule=UseColor_Specified, specifiedColor=<rgba 1 1 1 1>}
     uVRegion {bIsValid=false, max=<0 0>, min=<0 0>}
    }
   }
   slot {class=OverlaySlot, properties {horizontalAlignment=HAlign_Center}}
  }
  NumericText_Speed {
   widgetClassPath=/Script/CommonUI.CommonNumericTextBlock
   bInherited=true
   bIsVariable=true
   parent=Overlay_Speedometer
   widget=NumericText_Speed
   properties {
    bIsVolatile=true
    justification=Center
    performSizeInterpolation=false
    postInterpolationShrinkDuration=0
    text=Text Block
    font {
     bForceMonospaced=false
     bMaterialIsStencil=false
     fontMaterial=null
     fontObject=&Burbank
     letterSpacing=0
     monospacedWidth=1
     size=52.5
     skewAmount=0
     typefaceFontName=BurbankRegular-Black
     outlineSettings {
      bApplyOutlineToDropShadows=false
      bMiteredCorners=false
      bSeparateFillAlpha=false
      outlineMaterial=null
      outlineSize=0
      outlineColor=<rgba 0 0 0 1>
     }
    }
    renderTransform {angle=0, scale=<1 1>, shear=<-8 0>, translation=<0 -5>}
    shadowOffset=<0 0>
   }
   slot {class=OverlaySlot, properties {horizontalAlignment=HAlign_Center, verticalAlignment=VAlign_Center}}
  }
  CommonTextBlock_72 {
   widgetClassPath=&CommonTextBlock
   parent=Overlay_Speedometer
   widget=CommonTextBlock_72
   properties {
    text=KPH
    visibility=Hidden
    font {
     bForceMonospaced=false
     bMaterialIsStencil=false
     fontMaterial=null
     fontObject=&Burbank
     letterSpacing=0
     monospacedWidth=1
     size=15
     skewAmount=0
     typefaceFontName=BurbankRegular-Black
     outlineSettings {
      bApplyOutlineToDropShadows=false
      bMiteredCorners=false
      bSeparateFillAlpha=false
      outlineMaterial=null
      outlineSize=0
      outlineColor=<rgba 0 0 0 1>
     }
    }
    renderTransform {angle=0, scale=<1 1>, shear=<-8 0>, translation=<-3 47>}
    renderTransformPivot=<0.5 0>
    shadowOffset=<0 0>
   }
   slot {class=OverlaySlot, properties {horizontalAlignment=HAlign_Center, verticalAlignment=VAlign_Center}}
  }
  OverlayBoostMeter {
   widgetClassPath=&Overlay
   bIsVariable=true
   parent=Overlay_SpeedCluster
   widget=OverlayBoostMeter
   properties {visibility=Collapsed, renderTransform {angle=0, scale=<-1 1>, shear=<0 0>, translation=<32 -26>}}
   slot {class=OverlaySlot, properties {horizontalAlignment=HAlign_Center, verticalAlignment=VAlign_Center, padding=<ltrb -200 -200 -200 -200>}}
  }
  Image_BoostMeter {
   widgetClassPath=&Image
   bInherited=true
   bIsVariable=true
   parent=OverlayBoostMeter
   widget=Image_BoostMeter
   properties {brush {
     bIsDynamicallyLoaded=false
     drawAs=Image
     imageType=FullColor
     mirroring=NoMirror
     resourceName=None
     resourceObject=/Game/Athena/HUD/Vehicles/Materials/MI_UI_WhiplashBoostMeter
     tiling=NoTile
     imageSize=<400 400>
     margin=<ltrb 0 0 0 0>
     outlineSettings {
      bUseBrushTransparency=false
      roundingType=HalfHeightRadius
      width=0
      color {colorUseRule=UseColor_Specified, specifiedColor=<rgba 0 0 0 0>}
      cornerRadii=<0 0 0 0>
     }
     tintColor {colorUseRule=UseColor_Specified, specifiedColor=<rgba 1 1 1 1>}
     uVRegion {bIsValid=false, max=<0 0>, min=<0 0>}
    }}
   slot {class=OverlaySlot, properties {padding=<ltrb 107 18 0 0>}}
  }
  Image_BoostIcon {
   widgetClassPath=&Image
   parent=OverlayBoostMeter
   widget=Image_BoostIcon
   properties {
    brush {
     bIsDynamicallyLoaded=false
     drawAs=Image
     imageType=FullColor
     mirroring=NoMirror
     resourceName=None
     resourceObject=/Game/Athena/HUD/Vehicles/Materials/M_UI_ValetBoostIcon
     tiling=NoTile
     imageSize=<30 30>
     margin=<ltrb 0 0 0 0>
     outlineSettings {
      bUseBrushTransparency=false
      roundingType=HalfHeightRadius
      width=0
      color {colorUseRule=UseColor_Specified, specifiedColor=<rgba 0 0 0 0>}
      cornerRadii=<0 0 0 0>
     }
     tintColor {colorUseRule=UseColor_Specified, specifiedColor=<rgba 1 1 1 1>}
     uVRegion {bIsValid=false, max=<0 0>, min=<0 0>}
    }
    renderTransform {angle=0, scale=<1 1>, shear=<0 0>, translation=<105 -95>}
   }
   slot {class=OverlaySlot, properties {verticalAlignment=VAlign_Center, padding=<ltrb 45 -85 -64 -64>}}
  }
  Overlay_BoostSparks {
   widgetClassPath=&Overlay
   bInherited=true
   bIsVariable=true
   parent=OverlayBoostMeter
   widget=Overlay_BoostSparks
   properties {visibility=Collapsed, renderTransform {angle=33, scale=<1 1>, shear=<0 0>, translation=<-140 5>}, renderTransformPivot=<135 0.5>}
   slot {class=OverlaySlot, properties {horizontalAlignment=HAlign_Center, verticalAlignment=VAlign_Center}}
  }
  Overlay_47 {
   widgetClassPath=&Overlay
   parent=Overlay_BoostSparks
   widget=Overlay_47
   properties {visibility=HitTestInvisible, renderTransform {angle=-180, scale=<1 1>, shear=<0 0>, translation=<0 0>}}
   slot {class=OverlaySlot}
  }
  Spacer_42 {widgetClassPath=&Spacer, parent=Overlay_47, widget=Spacer_42, properties {visibility=HitTestInvisible}, slot {class=OverlaySlot}}
  Image_Pulse {
   widgetClassPath=&Image
   bIsVariable=true
   parent=Overlay_47
   widget=Image_Pulse
   properties {
    visibility=HitTestInvisible
    brush {
     bIsDynamicallyLoaded=false
     drawAs=Image
     imageType=NoImage
     mirroring=NoMirror
     resourceName=None
     resourceObject=/Game/Athena/HUD/LTM/Wax/Materials/M_UI_TextGlow_Pulse
     tiling=NoTile
     imageSize=<44.4227 97.4638>
     margin=<ltrb 0 0 0 0>
     outlineSettings {
      bUseBrushTransparency=false
      roundingType=HalfHeightRadius
      width=0
      color {colorUseRule=UseColor_Specified, specifiedColor=<rgba 0 0 0 0>}
      cornerRadii=<0 0 0 0>
     }
     tintColor {colorUseRule=UseColor_Specified, specifiedColor=<rgba 0.8906 0.1088 0 1>}
     uVRegion {bIsValid=false, max=<0 0>, min=<0 0>}
    }
    renderTransform {angle=0, scale=<1 1>, shear=<0 0>, translation=<0 8>}
   }
   slot {class=OverlaySlot, properties {horizontalAlignment=HAlign_Center, verticalAlignment=VAlign_Center, padding=<ltrb -200 -200 -200 -200>}}
  }
  Image_Glow {
   widgetClassPath=&Image
   bIsVariable=true
   parent=Overlay_47
   widget=Image_Glow
   properties {
    visibility=HitTestInvisible
    brush {
     bIsDynamicallyLoaded=false
     drawAs=Image
     imageType=NoImage
     mirroring=NoMirror
     resourceName=None
     resourceObject=/Game/Athena/HUD/Vehicles/Materials/M_UI_SparksGlow
     tiling=NoTile
     imageSize=<130 109.0002>
     margin=<ltrb 0 0 0 0>
     outlineSettings {
      bUseBrushTransparency=false
      roundingType=HalfHeightRadius
      width=0
      color {colorUseRule=UseColor_Specified, specifiedColor=<rgba 0 0 0 0>}
      cornerRadii=<0 0 0 0>
     }
     tintColor {colorUseRule=UseColor_Specified, specifiedColor=<rgba 1 1 1 1>}
     uVRegion {bIsValid=false, max=<0 0>, min=<0 0>}
    }
    renderTransform {angle=0, scale=<1 1>, shear=<0 0>, translation=<0 17>}
   }
   slot {class=OverlaySlot, properties {horizontalAlignment=HAlign_Center, verticalAlignment=VAlign_Center, padding=<ltrb -200 -200 -200 -200>}}
  }
  Image_Sparks {
   widgetClassPath=&Image
   bIsVariable=true
   parent=Overlay_47
   widget=Image_Sparks
   properties {
    visibility=HitTestInvisible
    brush {
     bIsDynamicallyLoaded=false
     drawAs=Image
     imageType=NoImage
     mirroring=NoMirror
     resourceName=None
     resourceObject=/Game/Athena/UI/InGame/Interrogation/Materials/M_UI_Starburst_Sparks
     tiling=NoTile
     imageSize=<295.3884 400>
     margin=<ltrb 0 0 0 0>
     outlineSettings {
      bUseBrushTransparency=false
      roundingType=HalfHeightRadius
      width=0
      color {colorUseRule=UseColor_Specified, specifiedColor=<rgba 0 0 0 0>}
      cornerRadii=<0 0 0 0>
     }
     tintColor {colorUseRule=UseColor_Specified, specifiedColor=<rgba 1 0.3682 0.0743 1>}
     uVRegion {bIsValid=false, max=<0 0>, min=<0 0>}
    }
   }
   slot {class=OverlaySlot, properties {horizontalAlignment=HAlign_Center, verticalAlignment=VAlign_Center, padding=<ltrb -200 -200 -200 -200>}}
  }
  Image_Sparks_1 {
   widgetClassPath=&Image
   bIsVariable=true
   parent=Overlay_47
   widget=Image_Sparks_1
   properties {
    visibility=HitTestInvisible
    brush {
     bIsDynamicallyLoaded=false
     drawAs=Image
     imageType=NoImage
     mirroring=NoMirror
     resourceName=None
     resourceObject=/Game/Athena/UI/InGame/Interrogation/Materials/M_UI_Starburst_Sparks_Inst
     tiling=NoTile
     imageSize=<120 400>
     margin=<ltrb 0 0 0 0>
     outlineSettings {
      bUseBrushTransparency=false
      roundingType=HalfHeightRadius
      width=0
      color {colorUseRule=UseColor_Specified, specifiedColor=<rgba 0 0 0 0>}
      cornerRadii=<0 0 0 0>
     }
     tintColor {colorUseRule=UseColor_Specified, specifiedColor=<rgba 1 0.7183 0.651 1>}
     uVRegion {bIsValid=false, max=<0 0>, min=<0 0>}
    }
   }
   slot {class=OverlaySlot, properties {horizontalAlignment=HAlign_Center, verticalAlignment=VAlign_Center, padding=<ltrb -200 -200 -200 -200>}}
  }
  Image_LargeCone {
   widgetClassPath=&Image
   bIsVariable=true
   parent=Overlay_47
   widget=Image_LargeCone
   properties {
    visibility=HitTestInvisible
    brush {
     bIsDynamicallyLoaded=false
     drawAs=Image
     imageType=NoImage
     mirroring=NoMirror
     resourceName=None
     resourceObject=/Game/Athena/HUD/Vehicles/Materials/M_UI_SparksGlow_Inst
     tiling=NoTile
     imageSize=<150 219.5743>
     margin=<ltrb 0 0 0 0>
     outlineSettings {
      bUseBrushTransparency=false
      roundingType=HalfHeightRadius
      width=0
      color {colorUseRule=UseColor_Specified, specifiedColor=<rgba 0 0 0 0>}
      cornerRadii=<0 0 0 0>
     }
     tintColor {colorUseRule=UseColor_Specified, specifiedColor=<rgba 1 1 1 1>}
     uVRegion {bIsValid=false, max=<0 0>, min=<0 0>}
    }
    renderTransform {angle=0, scale=<1 1>, shear=<0 0>, translation=<0 44>}
   }
   slot {class=OverlaySlot, properties {horizontalAlignment=HAlign_Center, verticalAlignment=VAlign_Center, padding=<ltrb -200 -200 -200 -200>}}
  }
  Overlay_CarInfo {
   widgetClassPath=&Overlay
   bIsVariable=true
   parent=Overlay_Gauge
   widget=Overlay_CarInfo
   properties {renderTransform {angle=0, scale=<1 1>, shear=<0 0>, translation=<-25 -22>}}
   slot {class=OverlaySlot, properties {verticalAlignment=VAlign_Bottom}}
  }
  Text_ItemName {
   widgetClassPath=&CommonTextBlock
   bInherited=true
   bIsVariable=true
   parent=Overlay_CarInfo
   widget=Text_ItemName
   properties {
    justification=Right
    text="[PH] Item Name"
    font {
     bForceMonospaced=false
     bMaterialIsStencil=false
     fontMaterial=null
     fontObject=&Burbank
     letterSpacing=0
     monospacedWidth=1
     size=20
     skewAmount=0
     typefaceFontName=Burbank-Black
     outlineSettings {
      bApplyOutlineToDropShadows=true
      bMiteredCorners=false
      bSeparateFillAlpha=true
      outlineMaterial=&M-FontShadow
      outlineSize=3
      outlineColor=<rgba 0 0 0 0.5>
     }
    }
    shadowOffset=<0 0>
   }
   slot {class=OverlaySlot}
  }
  Image_BoostGlow {widgetClassPath=&Image, bInherited=true, widget=null}
  TutorialHighlight {widgetClassPath=&TutorialHighlight}
 ]
}
